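# EQS query for GnatGala AI that finds teleport spots around its current target.
# Generates a few projected points on a narrow arc around the target, oriented by the target's rotation.
# A vision-cone test with line of sight keeps only points outside players' views.
# A random test, set to filter only, follows.
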EnvQuery /FNM_AI_GnatGala/EQS/EQS_GnatGala_Teleport
options [{
  generator {
   name=EnvQueryGenerator_OnCircle_0
   class=EnvQueryGenerator_OnCircle
   properties {
    arcAngle=70
    bDefineArc=true
    circleCenter=&FortAthenaAIBot_EQSQueryContext_CurrentTarget
    circleRadius=1000
    numberOfPoints=5
    arcDirection {dirMode=Rotation, rotation=&FortAthenaAIBot_EQSQueryContext_CurrentTarget}
    projectionData {extentX=100, navigationFilter=null, postProjectionVerticalOffset=100, projectDown=1024}
    traceData {navigationFilter=null}
   }
  }
  tests [
   FortQueryTest_IsInAnyPlayerVisionCone_0 {
    class=FortQueryTest_IsInAnyPlayerVisionCone
    properties {
     bDefineReferenceValue=false
     boolValue=false
     clampMaxType=None
     clampMinType=None
     fieldOfView=100
     filterType=Match
     floatValueMax=0
     floatValueMin=0
     maxDistance=3000
     multipleContextFilterOp=AllPass
     multipleContextScoreOp=AverageScore
     normalizationType=Absolute
     referenceValue=0
     scoreClampMax=0
     scoreClampMin=0
     scoringEquation=Constant
     scoringFactor=1
     testComment=""
     testPurpose=Filter
     useLineOfSight=true
    }
   }
   FortQueryTest_Random_1 {class=FortQueryTest_Random, properties {floatValueMax=1, scoringFactor=1, testPurpose=Filter}}
  ]
 }]
